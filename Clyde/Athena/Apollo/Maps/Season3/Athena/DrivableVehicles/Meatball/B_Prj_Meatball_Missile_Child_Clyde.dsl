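; Child Blueprint of the Meatball vehicle missile projectile that swaps in a Clyde water explosion effect and limits homing time.
; BeginPlay runs the parent logic, resets the flight timer and assigns the Clyde Niagara water explosion asset.
; Tick accumulates flight time and, from the maximum homing time onward, sets empty homing data through the projectile movement interface on every frame.
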(event EventBeginPlay
  (|CallParentFunction)
  (Variables|Default|SetFlightTime 0.0)
  (Variables|Default|SetFX_Water_Explosion_Asset "/Clyde/Athena/Apollo/Maps/Season3/Athena/DrivableVehicles/Meatball/NS_GrenadeLauncher_Explosion_Asteria_water_clyde.NS_GrenadeLauncher_Explosion_Asteria_water_clyde"))

(event EventTick
  (params (DeltaSeconds :type Float%20%28single-precision%29))
  (|CallParentFunction DeltaSeconds)
  (bind _output_get (Variables|Default|SetFlightTime (+ (Variables|Default|GetFlightTime) DeltaSeconds)))
  (if (>= _output_get (Variables|Default|GetMaxHomingFlightTime))
    (Projectile|SetHomingData 0 0 (Utilities|Struct|MakeProjectileHomingData "None" 0.0 0.0 0.0) :!class /Script/FortniteGame.FortProjectileMovementInterface)))

(fn ConstructionScript
  (|CallParentFunction))
