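# Niagara water effects for the Meatball drivable boat (Clyde variant), including speed lines, surface foam, side mesh splashes and motor burble.
# User variables for forward speed, water contact and pontoon data drive spawn rates, placement and material parameters.
# Foam is constrained to the mid pontoon's water plane, and burble takes the rear pontoon's water height.
# Splashes respond to boat roll, and one splash emitter varies spawning by camera angle through dot-product curves.
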NiagaraSystem /Clyde/Athena/Apollo/Maps/Season3/Athena/DrivableVehicles/Meatball/NS_Vehicle_Meatball_Water_Clyde
dependencies {
 usedDataInterfaces [
  &NiagaraDataInterfaceCurve
  &NiagaraDataInterfaceVector2DCurve
  &NiagaraDataInterfaceVectorCurve
  &NiagaraDataInterfaceColorCurve
  /Script/Water.NiagaraDataInterfaceWater
  &NiagaraDataInterfaceCamera
 ]
 usedDynamicInputs [/Game/Effects/Niagara/DynamicInputs/Math/FN_RemapDotProduct]
 usedModules [
  /Game/Athena/DrivableVehicles/Meatball/Meatball_Large/FX/Modules/FN_VehicleUserVariables
  /Game/Athena/DrivableVehicles/Meatball/Meatball_Large/FX/Modules/FN_SphericalPontoonUserVariables
  /Game/Athena/DrivableVehicles/Meatball/Meatball_Large/FX/Modules/Pontoons/FrontPontoonBig
  /Game/Athena/DrivableVehicles/Meatball/Meatball_Large/FX/Modules/Pontoons/RearPontoonBig
  /Game/Athena/DrivableVehicles/Meatball/Meatball_Large/FX/Modules/Pontoons/MidPontoonBig
  &self:SystemSpawnScript.SystemScriptSource.SystemScriptGraph.NiagaraNodeAssignment_0.SetVariables_#1
  &SystemState
 ]
}
emitters {
 FN_Sprite_SpeedLines {
  emitterUpdateScript [
   &EmitterState(Life Cycle Mode=::NewEnumerator1, Loop Duration=0.25, UseLoopDelay=false) {module=EmitterState001}
   &SpawnBurst_Instantaneous(Spawn Count=5, Use Spawn Probability=false) {enabled=false}
   &EmitterState()
   &SpawnRate(
    SpawnRate=&Lerp_Float(A=0, B=&ClampFloat(Float=180, Max=350), Alpha=&Power(a=$User.ForwardSpeed, b=6))
    Use Spawn Probability=false
   )
  ]
  particleSpawnScript [
   /Niagara/Modules/Spawn/Initialization/InitializeParticle(
    Color=&MakeLinearColor(
     R:NiagaraFloat=0.2292
     G:NiagaraFloat=0.1989
     B:NiagaraFloat=0.1838
     A:NiagaraFloat=&UniformRangedFloat(Minimum=0.2, Maximum=0.4)
    )
    Lifetime=&UniformRangedFloat(Minimum=0.3, Maximum=0.4)
    Module.Write Color=true
    Module.Write Lifetime=true
    Module.Write Mass=true
    Module.Write Material Random=true
    Module.Write Position=false
    Module.Write Scale=false
    Module.Write SpriteRotation=true
    Module.Write SpriteSize=true
    Position=&SimulationPosition()
    Sprite Rotation=&UniformRangedFloat(Minimum=0, Maximum=360)
    Sprite Size=&UniformRangedVector2D(Minimum=<1.25 25>, Maximum=<0.8 60>)
   )
   &TorusLocation(
    Align Torus With Axis=true
    Handle Radius=32
    Large Radius=200
    MaskSpawnGroup=false
    Module.Override Randomness=false
    Offset=&Multiply_VectorByFloat(Vector=$Engine.Owner.SystemXAxis, Float=40)
    Torus Axis=<1 0 0>
   )
   &AddVelocity(Velocity=&MakeVector(X=&UniformRangedFloat(Minimum=-50, Maximum=-100), Y=0, Z=0))
   &InheritVelocity(Module.Limit Speed=true, Velocity Scale=<-0.9 -0.9 -0.9>)
  ]
  particleUpdateScript [
   &ParticleState()
   &Color(Scale Alpha=&FloatFromCurve(FloatCurve=DataInterface() {propertyValues {Curve {keys [<at 0 0 RCIM_Cubic>,<at 0.5 1 RCIM_Cubic>,<at 1 0 RCIM_Cubic>]}}}))
   &SolveForcesAndVelocity(Clamp Velocity=false, Limit Acceleration=false)
   &ScaleSpriteSizeBySpeed(Max Scale Factor=<5 10>, Min Scale Factor=<0 0.1>)
   &ScaleSpriteSize(Scale Factor=&Vector2DFromCurve(
     Vector2Curve=DataInterface() {propertyValues {
       XCurve {keys [<at -0.0069 0.5 RCIM_Cubic>,<at 0.4918 1 RCIM_Cubic>,<at 1 0.5 RCIM_Cubic>]}
       YCurve {keys [<at 0 0.0245 RCIM_Cubic>, {arriveTangent=-0.0245, interpMode=RCIM_Cubic, leaveTangent=-0.0245, time=0.4975, value=1}, <at 1 0 RCIM_Cubic>]}
      }}
     ScaleCurve=<1 1>
    ))
   &CameraOffset(Camera Offset Amount=-100)
  ]
  rendererClasses [&NiagaraSpriteRendererProperties]
  renderers [{rendererClass=&NiagaraSpriteRendererProperties}]
 }
 FN_Water_Foam_SnapToSurface_NoSkelRef {
  emitterUpdateScript [
   &EmitterState(Life Cycle Mode=::NewEnumerator1, Loop Duration=0.25, UseLoopDelay=false) {module=EmitterState001}
   &EmitterState() {enabled=false}
   &EmitterScalability(
    MaxDistance=2000
    Module.bApplySpawnCountScale=false
    Module.bApplySpawnCountScaleByDistanceFraction=false
    Module.bUseMaxDistance=false
    Module.bUseMinDistance=false
    SpawnCountScaleByDistanceFraction=DataInterface() {propertyValues {Curve {keys [<at 0 1>,<at 1 1>]}}}
   ) {enabled=false}
   &SpawnPerUnit(Module.Use Max Movement Threshold=true, Module.Use Movement Tolerance=false, Spawn Per Unit=150, Use Spawn Probability=false) {enabled=false}
   &EmitterLifeCycle(NextLoopDuration=0.25) {enabled=false}
   &SpawnRate(
    SpawnRate=&MakeCustomFloatFromBool(
     Bool=&Boolean_AndOperation(
      Bool1=$User.InWater
      Bool2=/Game/Athena/DrivableVehicles/Meatball/Meatball_Large/FX/DynamicInputs/FN_SpawnFromPontoon(PontoonLocation:Vector3f=Linked() {linkedVariable {name=User.MidPontoonBig.Location, type=NiagaraPosition}}, Threshold:NiagaraFloat=100)
     )
     True Float=&Lerp_Float(A=1, B=7, Alpha=$User.ForwardSpeed)
     False Float=0
    )
    Use Spawn Probability=false
   )
  ]
  particleSpawnScript [
   &InitializeParticle(
    Lifetime=&UniformRangedFloat(Minimum=2.4, Maximum=3)
    Mass=&UniformRangedFloat(Minimum=0.6, Maximum=1.2)
    Mesh Scale=&UniformRangedVector(Minimum=<25 25 25>, Maximum=<30 35 35>)
    Module.Write Color=true
    Module.Write Lifetime=true
    Module.Write Mass=true
    Module.Write Material Random=true
    Module.Write Position=false
    Module.Write Scale=true
    Module.Write SpriteRotation=false
    Module.Write SpriteSize=false
    Position=&SimulationPosition()
    Sprite Rotation=&NormalizedAngleToDegrees(NormalizedAngle=&UniformRangedFloat(Minimum=0, Maximum=1))
   ) {enabled=false}
   /Niagara/Modules/Spawn/Initialization/V2/InitializeParticle(
    Lifetime=&UniformRangedFloat(Minimum=2, Maximum=3)
    Mass=&UniformRangedFloat(Minimum=0.6, Maximum=1.2)
    Mass Mode=::NewEnumerator1
    Mesh Scale Mode=::NewEnumerator1
    Mesh Uniform Scale Max=25
    Mesh Uniform Scale Min=23
    UsePositionOffset=false
    Write Material Random=true
   ) {module=InitializeParticle001}
   /SetVariables(Particles.MeshOrientation=/Game/Athena/DrivableVehicles/Meatball/Meatball_Large/FX/DynamicInputs/FN_OwnerRotationRollandPitchRemoved()) {isSetParameters=true}
   /SetVariables(Particles.Velocity=&Multiply_VectorByFloat(Vector=$Engine.Owner.SystemXAxis, Float=&Lerp_Float(A=0, B=-800, Alpha=$User.ForwardSpeed))) {enabled=false, isSetParameters=true}
   /Niagara/Modules/Spawn/Orientation/V2/InitialMeshOrientation(
    Orientation Vector=&RotateVectorByQuaternion(Vector To Rotate=<1 0 0>, Quaternion=$System.AlignQuaternion.Quat)
    Rotation=&UniformRangedVector(Minimum=<0 0 0>, Maximum=<0 0 0>)
    Use Orientation Vector=true
    Use Rotation Vector=true
   ) {enabled=false}
   /Niagara/Modules/Spawn/Orientation/V3/InitialMeshOrientation(
    Mesh Orientation Mode=::NewEnumerator2
    Rotation=&UniformRangedVector(Minimum=<0 0 0>, Maximum=<0 0 1>)
    Use Rotation Vector=false
   ) {module=InitialMeshOrientation001}
   &DynamicMaterialParameters(
    Index 1 Param 3=200
    Index 3 Param 1=$User.MidPontoonBig.WaterBodyIndex
    Index 3 Param 2=$User.MidPontoonBig.Depth
    Module.Param0WriteEnabled1=false
    Module.Param0WriteEnabled3=true
    Module.Param1WriteEnabled1=false
    Module.Param1WriteEnabled3=true
    Module.Param2WriteEnabled1=true
    Module.Param2WriteEnabled3=true
    Module.Param3WriteEnabled1=false
    Module.Param3WriteEnabled3=true
    Write Parameter Index 0=false
    Write Parameter Index 1=true
    Write Parameter Index 3=true
   ) {enabled=false}
   /SetVariables(Particles.WaterPlaneLocation:Vector3f=$User.MidPontoonBig.WaterPlaneLocation) {isSetParameters=true}
   /SetVariables(Particles.WaterPlaneNormal:Vector3f=$User.MidPontoonBig.WaterPlaneNormal) {isSetParameters=true}
   /SetVariables(Particles.Position=$User.MidPontoonBig.Location) {isSetParameters=true}
   &InitializeWaterSnapToOwner()
   &InheritVelocity(Module.Limit Speed=true, Velocity Scale=<-1 -1 -1>)
  ]
  particleUpdateScript [
   &ParticleState()
   &DynamicMaterialParameters(
    Index 0 Param 1=&RandomRangeFloat(Minimum=-1, Maximum=1)
    Index 0 Param 2=&FloatFromCurve(FloatCurve=DataInterface() {propertyValues {Curve {keys [<at 0 0>,<at 1 1.25>]}}})
    Index 0 Param 3=&FloatFromCurve(
     FloatCurve=DataInterface() {propertyValues {Curve {keys [{interpMode=RCIM_Cubic, tangentMode=RCTM_User, time=0, value=1}, {interpMode=RCIM_Cubic, tangentMode=RCTM_User, time=1, value=0}]}}}
     Scale Curve=0.66
    )
    Index 1 Param 1=&MakeFloatFromVector(Vector=Linked() {linkedVariable {name=Particles.Position, type=NiagaraPosition}}, Channel=::NewEnumerator0)
    Index 1 Param 2=&MakeFloatFromVector(Vector=Linked() {linkedVariable {name=Particles.Position, type=NiagaraPosition}}, Channel=::NewEnumerator1)
    Index 1 Param 3=&FloatFromCurve(
     FloatCurve=DataInterface() {propertyValues {Curve {keys [<at 0 1 RCIM_Cubic>, {arriveTangent=0.4219, interpMode=RCIM_Cubic, leaveTangent=0.4219, tangentMode=RCTM_User, time=1, value=1.25}]}}}
     Scale Curve=225
    )
    Index 2 Param 1=&FloatFromCurve(FloatCurve=DataInterface() {propertyValues {Curve {keys [{interpMode=RCIM_Cubic, tangentMode=RCTM_User, time=0.1, value=0}, {interpMode=RCIM_Cubic, tangentMode=RCTM_User, time=1, value=1}]}}})
    Index 2 Param 2=&FloatFromCurve(
     FloatCurve=DataInterface() {propertyValues {Curve {keys [<at -0.0046 0 RCIM_Cubic>,<at 0.11 0.9612 RCIM_Cubic>,<at 1.0011 0>]}}}
     Scale Curve=2
    )
    Index 2 Param 3=1
    Index 2 Param 4=&FloatFromCurve(FloatCurve=DataInterface() {propertyValues {Curve {keys [
         {arriveTangent=-1.9991, interpMode=RCIM_Cubic, time=0, value=0.5066}
         {arriveTangent=-1.0723, interpMode=RCIM_Cubic, leaveTangent=-1.0723, tangentMode=RCTM_User, time=0.2095, value=0.2954}
         <at 1 0 RCIM_Cubic>
        ]}}})
    Module.Param0WriteEnabled=true
    Module.Param0WriteEnabled1=true
    Module.Param0WriteEnabled2=true
    Module.Param1WriteEnabled=true
    Module.Param1WriteEnabled1=true
    Module.Param1WriteEnabled2=true
    Module.Param2WriteEnabled=true
    Module.Param2WriteEnabled1=true
    Module.Param2WriteEnabled2=true
    Module.Param3WriteEnabled=true
    Module.Param3WriteEnabled1=true
    Module.Param3WriteEnabled2=true
    Write Parameter Index 1=true
    Write Parameter Index 2=true
   ) {module=DynamicMaterialParameters001}
   &ScaleColor(
    Scale Alpha=&FloatFromCurve(FloatCurve=DataInterface() {propertyValues {Curve {keys [
         {arriveTangent=37.6118, interpMode=RCIM_Cubic, leaveTangent=37.6118, tangentMode=RCTM_User, time=0, value=0}
         {arriveTangent=0.002, interpMode=RCIM_Cubic, leaveTangent=0.002, time=0.0961, value=0.9983}
         <at 1.0011 0.002 RCIM_Cubic>
        ]}}})
    ScaleA=true
    ScaleRGB=true
   ) {enabled=false}
   &Drag(Drag=2, Use Linear Drag=true, Use Rotational Drag=false)
   &SolveForcesAndVelocity(Clamp Velocity=false, Limit Acceleration=false)
   &ScaleMeshSize(Scale Factor=&Multiply_Vector(
     A=$Particles.Initial.Scale
     B=&VectorFromCurve(VectorCurve=DataInterface() {propertyValues {
        XCurve {keys [{arriveTangent=1.7356, interpMode=RCIM_Cubic, leaveTangent=1.7356, tangentMode=RCTM_User, time=-0.0011, value=0.3088}, <at 1 1>]}
        YCurve {keys [
          {arriveTangent=0.5774, interpMode=RCIM_Cubic, leaveTangent=0.5774, tangentMode=RCTM_User, time=-0.0011, value=0.313}
          {arriveTangent=2.0637, interpMode=RCIM_Cubic, leaveTangent=2.0638, tangentMode=RCTM_User, time=1, value=1.25}
         ]}
        ZCurve {keys [{arriveTangent=1.6409, interpMode=RCIM_Cubic, leaveTangent=1.6409, tangentMode=RCTM_User, time=-0.0011, value=0.313}, <at 1 1>]}
       }})
    )) {module=MeshSizeScale}
   &ConstrainPositionToPlane(Plane Normal:Vector3f=$Particles.WaterPlaneNormal, Plane Position:Vector3f=$Particles.WaterPlaneLocation)
  ]
  rendererClasses [&NiagaraMeshRendererProperties]
  renderers [{rendererClass=&NiagaraMeshRendererProperties, rendererIndex=0}]
 }
 Mesh_Splashes_LS {
  emitterUpdateScript [
   &EmitterState(Life Cycle Mode=::NewEnumerator1, Loop Duration=0.25, UseLoopDelay=false)
   &EmitterScalability(
    Module.bApplySpawnCountScale=false
    Module.bApplySpawnCountScaleByDistanceFraction=false
    Module.bUseMaxDistance=false
    Module.bUseMinDistance=false
   ) {enabled=false}
   &EmitterLifeCycle(NextLoopDuration=0.25) {enabled=false}
   &SpawnRate(
    SpawnRate=&MakeCustomFloatFromBool(Bool=$User.InWater, True Float=&Lerp_Float(A=0, B=320, Alpha=$User.ForwardSpeed), False Float=0)
    Use Spawn Probability=false
   )
  ]
  particleSpawnScript [
   /Game/Athena/DrivableVehicles/Meatball/Meatball_Large/FX/Modules/FN_BoatInitialize(Lifetime:NiagaraFloat=&Lerp_Float(
     A=&UniformRangedFloat(Minimum=0.254, Maximum=0.55)
     B=&UniformRangedFloat(Minimum=0.325, Maximum=0.675)
     Alpha=$User.ForwardSpeed
    ))
   /Game/Athena/DrivableVehicles/Meatball/Meatball_Large/FX/Modules/V2/FN_LinePosition(
    End=165
    Offset Direction:Vector3f=&MakeVector(
     X=0
     Y=&Lerp_Float(
      A=&FloatFromCurve(
       FloatCurve=DataInterface() {propertyValues {Curve {keys [<at -0.007 0.8267>,<at 0.9965 0.6159>]}}}
       CurveIndex=$Particles.RandomDistribution
       Scale Curve=150
      )
      B=100
      Alpha=HlslExpression() {hlslExpression="min(abs(System.BoatRoll)*3,1)*max((Particles.BoatSign*sign(System.BoatRoll)),1)"}
     )
     Z=0
    )
    Start=-120
    Symmetry=true
   )
   &BoxLocation(Box Offset=<0 0 18>, Box Size=<0 0 0>, MaskSpawnGroup=false, Module.Override Seed=false, Rotate Box=false, Surface Only=false)
   /Game/Athena/DrivableVehicles/Meatball/Meatball_Large/FX/Modules/FN_VehicleScale(
    Boost Multiplier=0.125
    Mesh Scale:Vector3f=<1 2 1>
    Modulate by Pontoon Depth=false
    Modulate by Pontoon Impulse=false
    Modulate by Random Float=true
    Modulate by Vehicle Speed=true
    Module.Write Mesh Scale=true
    Module.Write Sprite Size=false
    Remap Forward Speed=DataInterface() {propertyValues {Curve {keys [{arriveTangent=3.1047, interpMode=RCIM_Cubic, leaveTangent=3.1047, tangentMode=RCTM_User, time=0.121, value=-0.0129}, <at 1.0045 1.0505>]}}}
    SizeCurve=DataInterface() {propertyValues {Curve {keys [
        {arriveTangent=-0.6878, interpMode=RCIM_Cubic, leaveTangent=-0.6878, tangentMode=RCTM_User, time=0, value=1}
        {arriveTangent=-0.7009, interpMode=RCIM_Cubic, leaveTangent=-0.7009, tangentMode=RCTM_User, time=0.1862, value=0.615}
        {arriveTangent=-0.5072, interpMode=RCIM_Cubic, leaveTangent=-0.5072, tangentMode=RCTM_User, time=1, value=0.25}
       ]}}}
    Sprite Size:Vector2f=<100 250>
   )
   /Game/Athena/DrivableVehicles/Meatball/Meatball_Large/FX/Modules/FN_VehicleDirectionVector(
    Rotate on X Axis=true
    Rotate on Y Axis=true
    Rotate on Z Axis=true
    Set MeshOrientation=true
    Set Sprite Alignment=false
    X Rotation End:NiagaraFloat=&UniformRangedFloat(Minimum=0.06, Maximum=0.1)
    Y Rotation End=-0.0125
    Y Rotation Start=0.005
    Z Rotation End:NiagaraFloat=&Lerp_Float(
     A=&UniformRangedFloat(Minimum=-0.35, Maximum=-0.55)
     B=&UniformRangedFloat(Minimum=0.35, Maximum=0.55)
     Alpha=$Particles.BoatSide
    )
    Z Rotation Start:NiagaraFloat=&Lerp_Float(A=-0.35, B=0.35, Alpha=$Particles.BoatSide)
   )
  ]
  particleUpdateScript [
   &ParticleState()
   /Game/Athena/DrivableVehicles/Meatball/Meatball_Large/FX/Modules/FN_BoatRemoveOwnerRoll_v2(
    Angle:NiagaraFloat=&FloatFromCurve(
     FloatCurve=DataInterface() {propertyValues {Curve {keys [<at -0.0058 0 RCIM_Cubic>, {arriveTangent=2.6901, interpMode=RCIM_Cubic, leaveTangent=2.6901, tangentMode=RCTM_User, time=1.0011, value=1}]}}}
     Scale Curve=0
    )
    Axis:ENiagaraOrientationAxis=::YAxis
    Offset:NiagaraFloat=&FloatFromCurve(FloatCurve=DataInterface() {propertyValues {Curve {keys [<at -0.0077 0.0713>,<at 0.9942 1.5508>]}}}, Scale Curve=-20)
    Offset position along X=true
    Remove Owner Pitch and Roll from Mesh Orientation=true
    Remove Owner Pitch and Roll from Sprite Alignment=true
   )
   /Game/Athena/DrivableVehicles/Meatball/Meatball_Large/FX/Modules/FN_ScaleByVehicleData(Boost Multiplier=0.2, Inside Roll Range Max=0.3, Outside Bank Multiplier=1.5)
   &SolveForcesAndVelocity(Clamp Velocity=false, Limit Acceleration=false)
   &DynamicMaterialParameters(
    Index 0 Param 1=&FloatFromCurve(FloatCurve=DataInterface() {propertyValues {Curve {keys [
         {arriveTangent=1.2629, interpMode=RCIM_Cubic, leaveTangent=1.2629, tangentMode=RCTM_User, time=-0.0045, value=-1.9799}
         {arriveTangent=4.4591, interpMode=RCIM_Cubic, leaveTangent=4.4592, tangentMode=RCTM_User, time=1, value=0}
        ]}}})
    Index 0 Param 2=&FloatFromCurve(FloatCurve=DataInterface() {propertyValues {Curve {keys [<at 0.0102 4>,<at 0.9749 0>]}}})
    Index 0 Param 3=&MakeFloatFromVector(Vector=$Particles.Scale, Channel=::NewEnumerator0)
    Index 0 Param 4=1
    Index 1 Param 1=&FloatFromCurve(FloatCurve=DataInterface() {propertyValues {Curve {keys [<at 0.7 0 RCIM_Cubic>, {arriveTangent=6.4641, interpMode=RCIM_Cubic, leaveTangent=6.4641, tangentMode=RCTM_User, time=1, value=1}]}}})
    Module.Param0WriteEnabled=true
    Module.Param0WriteEnabled1=true
    Module.Param1WriteEnabled=true
    Module.Param1WriteEnabled1=true
    Module.Param2WriteEnabled=true
    Module.Param2WriteEnabled1=true
    Module.Param3WriteEnabled=true
    Module.Param3WriteEnabled1=true
    Write Parameter Index 1=true
   )
  ]
  rendererClasses [&NiagaraMeshRendererProperties]
  renderers [
   {rendererClass=&NiagaraMeshRendererProperties, rendererIndex=0}
   {rendererClass=&NiagaraMeshRendererProperties, rendererIndex=1}
  ]
 }
 Mesh_Splashes_LS001 {
  emitterUpdateScript [
   &EmitterState(Life Cycle Mode=::NewEnumerator1, Loop Duration=0.25, UseLoopDelay=false)
   &EmitterScalability(
    Module.bApplySpawnCountScale=false
    Module.bApplySpawnCountScaleByDistanceFraction=false
    Module.bUseMaxDistance=false
    Module.bUseMinDistance=false
   ) {enabled=false}
   &EmitterLifeCycle(NextLoopDuration=0.25) {enabled=false}
   /SetVariables(
    Emitter.BoatDot:NiagaraFloat=&DotProduct(
     Vector A=&TransformVector(Vector=$Engine.Owner.SystemXAxis, Source Space=::World, Destination Space=::Simulation)
     Vector B=&NormalizeVector(
      Vector To Normalize=&ReturnCameraPosition(Coordinate Space:ENiagaraCoordinateSpace=::Local, Camera Query:NiagaraDataInterfaceCamera=DataInterface())
      Safe Normalize=true
      Safe Fallback Threshold=0
      Safe Fallback Vector=<0 0 0.001>
     )
    )
    Emitter.BoatDotTop:NiagaraFloat=&DotProduct(
     Vector A=<0 0 1>
     Vector B=&ReturnCameraVector(Camera Vectors=::NewEnumerator4, Coordinate Space=::World, Origin Position=/Niagara/DynamicInputs/Helpers/SimulationPosition())
    )
   ) {isSetParameters=true}
   &SpawnRate(
    SpawnRate=&MakeCustomFloatFromBool(
     Bool=$User.InWater
     True Float=&Lerp_Float(
      A=0
      B=&Lerp_Float(
       A=320
       B=-30
       Alpha=&Lerp_Float(
        A=&Lerp_Float(
         A=1
         B=0
         Alpha=&FloatFromCurve(
          FloatCurve=DataInterface() {propertyValues {Curve {keys [
              {arriveTangent=11.6483, arriveTangentWeight=0.8468, interpMode=RCIM_Cubic, leaveTangentWeight=0.8468, time=-1, value=0.2}
              {
               arriveTangentWeight=0.3595
               interpMode=RCIM_Cubic
               leaveTangentWeight=0.3595
               tangentMode=RCTM_User
               tangentWeightMode=RCTWM_WeightedBoth
               time=-0.25
               value=1
              }
             ]}}}
          CurveIndex=$Emitter.BoatDot
         )
        )
        B=0
        Alpha=&FloatFromCurve(
         FloatCurve=DataInterface() {propertyValues {Curve {keys [
             {arriveTangent=-5.0159, interpMode=RCIM_Cubic, leaveTangent=-5.0159, tangentMode=RCTM_User, time=-0.35, value=1}
             <at 0 0 RCIM_Cubic>
             {arriveTangent=2.1171, interpMode=RCIM_Cubic, leaveTangent=2.1171, tangentMode=RCTM_User, time=1, value=1}
            ]}}}
         CurveIndex=$Emitter.BoatDotTop
        )
       )
      )
      Alpha=$User.ForwardSpeed
     )
     False Float=0
    )
    Use Spawn Probability=false
   )
  ]
  particleSpawnScript [
   &FN_BoatInitialize(Lifetime:NiagaraFloat=&Lerp_Float(
     A=&UniformRangedFloat(Minimum=0.254, Maximum=0.55)
     B=&UniformRangedFloat(Minimum=0.325, Maximum=0.675)
     Alpha=$User.ForwardSpeed
    ))
   /SetVariables(Particles.RandomDistribution:NiagaraFloat=&Lerp_Float(
     A=&RandomRangeFloat(Minimum=0, Maximum=0)
     B=&RandomRangeFloat(Minimum=0, Maximum=1)
     Alpha=&Lerp_Float(
      A=&Lerp_Float(
       A=&FloatFromCurve(
        FloatCurve=DataInterface() {propertyValues {Curve {keys [<at 0 1 RCIM_Cubic>, {arriveTangent=-6.0459, interpMode=RCIM_Cubic, leaveTangent=-6.0459, tangentMode=RCTM_User, time=0.35, value=0}]}}}
        CurveIndex=$Particles.RandomDistribution
       )
       B=1
       Alpha=&FloatFromCurve(
        FloatCurve=DataInterface() {propertyValues {Curve {keys [
            {arriveTangent=11.6483, arriveTangentWeight=0.8468, interpMode=RCIM_Cubic, leaveTangentWeight=0.8468, time=-1, value=0.2}
            {
             arriveTangentWeight=0.3595
             interpMode=RCIM_Cubic
             leaveTangentWeight=0.3595
             tangentMode=RCTM_User
             tangentWeightMode=RCTWM_WeightedBoth
             time=-0.25
             value=1
            }
           ]}}}
        CurveIndex=$Emitter.BoatDot
       )
      )
      B=1
      Alpha=&FloatFromCurve(
       FloatCurve=DataInterface() {propertyValues {Curve {keys [
           {arriveTangent=-5.0159, interpMode=RCIM_Cubic, leaveTangent=-5.0159, tangentMode=RCTM_User, time=-0.35, value=1}
           <at 0 0 RCIM_Cubic>
           {arriveTangent=2.1171, interpMode=RCIM_Cubic, leaveTangent=2.1171, tangentMode=RCTM_User, time=1, value=1}
          ]}}}
       CurveIndex=$Emitter.BoatDotTop
      )
     )
    )) {isSetParameters=true}
   &FN_LinePosition(
    End=165
    Offset Direction:Vector3f=&MakeVector(
     X=0
     Y=&Lerp_Float(
      A=&FloatFromCurve(
       FloatCurve=DataInterface() {propertyValues {Curve {keys [<at -0.007 0.8267>,<at 0.9965 0.6159>]}}}
       CurveIndex=$Particles.RandomDistribution
       Scale Curve=150
      )
      B=100
      Alpha=HlslExpression() {hlslExpression="min(abs(System.BoatRoll)*3,1)*max((Particles.BoatSign*sign(System.BoatRoll)),1)"}
     )
     Z=0
    )
    Start=-120
    Symmetry=true
   )
   &BoxLocation(Box Offset=<0 0 18>, Box Size=<0 0 0>, MaskSpawnGroup=false, Module.Override Seed=false, Rotate Box=false, Surface Only=false)
   &FN_VehicleScale(
    Boost Multiplier=0.125
    Mesh Scale:Vector3f=<1 2 1>
    Modulate by Pontoon Depth=false
    Modulate by Pontoon Impulse=false
    Modulate by Random Float=true
    Modulate by Vehicle Speed=true
    Module.Write Mesh Scale=true
    Module.Write Sprite Size=false
    Remap Forward Speed=DataInterface() {propertyValues {Curve {keys [{arriveTangent=3.1047, interpMode=RCIM_Cubic, leaveTangent=3.1047, tangentMode=RCTM_User, time=0.121, value=-0.0129}, <at 1.0045 1.0505>]}}}
    SizeCurve=DataInterface() {propertyValues {Curve {keys [
        {arriveTangent=-0.6878, interpMode=RCIM_Cubic, leaveTangent=-0.6878, tangentMode=RCTM_User, time=0, value=1}
        {arriveTangent=-0.7009, interpMode=RCIM_Cubic, leaveTangent=-0.7009, tangentMode=RCTM_User, time=0.1862, value=0.615}
        {arriveTangent=-0.5072, interpMode=RCIM_Cubic, leaveTangent=-0.5072, tangentMode=RCTM_User, time=1, value=0.25}
       ]}}}
    Sprite Size:Vector2f=<100 250>
   )
   &FN_VehicleDirectionVector(
    Rotate on X Axis=true
    Rotate on Y Axis=true
    Rotate on Z Axis=true
    Set MeshOrientation=true
    Set Sprite Alignment=false
    X Rotation End:NiagaraFloat=&UniformRangedFloat(Minimum=0.06, Maximum=0.1)
    Y Rotation End=-0.0125
    Y Rotation Start=0.005
    Z Rotation End:NiagaraFloat=&Lerp_Float(
     A=&UniformRangedFloat(Minimum=-0.35, Maximum=-0.55)
     B=&UniformRangedFloat(Minimum=0.35, Maximum=0.55)
     Alpha=$Particles.BoatSide
    )
    Z Rotation Start:NiagaraFloat=&Lerp_Float(A=-0.35, B=0.35, Alpha=$Particles.BoatSide)
   )
   &KillParticles(Kill Particles=&SetBoolByFloatComparison(
     A=$Particles.RandomDistribution
     B=&FloatFromCurve(
      FloatCurve=DataInterface() {propertyValues {Curve {keys [<at 0 1>,<at 1 0>]}}}
      CurveIndex=&Abs_Float(Float:NiagaraFloat=&DotProduct(Vector A=<1 0 0>, Vector B=<1 0 0>))
     )
    )) {enabled=false}
   &Color(Scale Alpha=1.5)
  ]
  particleUpdateScript [
   &ParticleState()
   &ScaleMeshSize(Scale Factor=&VectorFromFloat(Value=&MakeCustomFloatFromBool(
      Bool=&SetBoolByFloatComparison(
       A=$Particles.RandomDistribution
       B=&FloatFromCurve(
        FloatCurve=DataInterface() {propertyValues {Curve {keys [{tangentMode=RCTM_User, time=0, value=1}, {tangentMode=RCTM_User, time=0.6, value=0}]}}}
        CurveIndex=&Abs_Float(Float:NiagaraFloat=&DotProduct(
          Vector A=$Engine.Owner.SystemXAxis
          Vector B=&NormalizeVector(
           Vector To Normalize=&Subtract_Position(
            A=&ReturnCameraPosition(Coordinate Space:ENiagaraCoordinateSpace=::Simulation, Camera Query:NiagaraDataInterfaceCamera=DataInterface())
            B=&SimulationPosition()
           )
           Safe Normalize=true
           Safe Fallback Threshold=0
           Safe Fallback Vector=<0 0 0.001>
          )
         ))
        Scale Curve=500
       )
      )
      True Float=1
      False Float=0
     ))) {enabled=false}
   &FN_BoatRemoveOwnerRoll_v2(
    Angle:NiagaraFloat=&FloatFromCurve(
     FloatCurve=DataInterface() {propertyValues {Curve {keys [<at -0.0058 0 RCIM_Cubic>, {arriveTangent=2.6901, interpMode=RCIM_Cubic, leaveTangent=2.6901, tangentMode=RCTM_User, time=1.0011, value=1}]}}}
     Scale Curve=0
    )
    Axis:ENiagaraOrientationAxis=::YAxis
    Offset:NiagaraFloat=&FloatFromCurve(FloatCurve=DataInterface() {propertyValues {Curve {keys [<at -0.0077 0.0713>,<at 0.9942 1.5508>]}}}, Scale Curve=-20)
    Offset position along X=true
    Remove Owner Pitch and Roll from Mesh Orientation=true
    Remove Owner Pitch and Roll from Sprite Alignment=true
   )
   &FN_ScaleByVehicleData(Boost Multiplier=0.2, Inside Roll Range Max=0.3, Outside Bank Multiplier=1.5)
   &SolveForcesAndVelocity(Clamp Velocity=false, Limit Acceleration=false)
   &DynamicMaterialParameters(
    Index 0 Param 1=&FloatFromCurve(FloatCurve=DataInterface() {propertyValues {Curve {keys [
         {arriveTangent=1.2629, interpMode=RCIM_Cubic, leaveTangent=1.2629, tangentMode=RCTM_User, time=-0.0045, value=-1.9799}
         {arriveTangent=4.4591, interpMode=RCIM_Cubic, leaveTangent=4.4592, tangentMode=RCTM_User, time=1, value=0}
        ]}}})
    Index 0 Param 2=&FloatFromCurve(FloatCurve=DataInterface() {propertyValues {Curve {keys [<at 0.0102 4>,<at 0.9749 0>]}}})
    Index 0 Param 3=&MakeFloatFromVector(Vector=$Particles.Scale, Channel=::NewEnumerator0)
    Index 0 Param 4=1
    Index 1 Param 1=&FloatFromCurve(FloatCurve=DataInterface() {propertyValues {Curve {keys [<at 0.7 0 RCIM_Cubic>, {arriveTangent=6.4641, interpMode=RCIM_Cubic, leaveTangent=6.4641, tangentMode=RCTM_User, time=1, value=1}]}}})
    Module.Param0WriteEnabled=true
    Module.Param0WriteEnabled1=true
    Module.Param1WriteEnabled=true
    Module.Param1WriteEnabled1=true
    Module.Param2WriteEnabled=true
    Module.Param2WriteEnabled1=true
    Module.Param3WriteEnabled=true
    Module.Param3WriteEnabled1=true
    Write Parameter Index 1=true
   )
   /SetVariables(Particles.ScaleDot:Vector3f=&Multiply_VectorByFloat(
     Vector=$Particles.Scale
     Float=&Lerp_Float(
      A=&Lerp_Float(
       A=&FloatFromCurve(
        FloatCurve=DataInterface() {propertyValues {Curve {keys [<at 0 1 RCIM_Cubic>, {arriveTangent=-6.0459, interpMode=RCIM_Cubic, leaveTangent=-6.0459, tangentMode=RCTM_User, time=0.35, value=0}]}}}
        CurveIndex=$Particles.RandomDistribution
       )
       B=1
       Alpha=&FloatFromCurve(
        FloatCurve=DataInterface() {propertyValues {Curve {keys [
            {arriveTangent=11.6483, arriveTangentWeight=0.8468, interpMode=RCIM_Cubic, leaveTangentWeight=0.8468, time=-1, value=0.2}
            {
             arriveTangentWeight=0.3595
             interpMode=RCIM_Cubic
             leaveTangentWeight=0.3595
             tangentMode=RCTM_User
             tangentWeightMode=RCTWM_WeightedBoth
             time=-0.25
             value=1
            }
           ]}}}
        CurveIndex=$Emitter.BoatDot
       )
      )
      B=1
      Alpha=&FloatFromCurve(
       FloatCurve=DataInterface() {propertyValues {Curve {keys [
           {arriveTangent=-5.0159, interpMode=RCIM_Cubic, leaveTangent=-5.0159, tangentMode=RCTM_User, time=-0.35, value=1}
           <at 0 0 RCIM_Cubic>
           {arriveTangent=2.1171, interpMode=RCIM_Cubic, leaveTangent=2.1171, tangentMode=RCTM_User, time=1, value=1}
          ]}}}
       CurveIndex=$Emitter.BoatDotTop
      )
     )
    )) {isSetParameters=true}
  ]
  rendererClasses [&NiagaraMeshRendererProperties]
  renderers [
   {rendererClass=&NiagaraMeshRendererProperties, rendererIndex=0}
   {rendererClass=&NiagaraMeshRendererProperties, rendererIndex=1}
  ]
 }
 Motor_Burble_Mesh_WS_NoSkelRef {
  emitterUpdateScript [
   &EmitterState(Life Cycle Mode=::NewEnumerator1, Loop Duration=0.25, UseLoopDelay=false)
   &EmitterScalability(
    MaxCulledState:ENiagaraExecutionState=::Active
    Module.bApplySpawnCountScale=false
    Module.bApplySpawnCountScaleByDistanceFraction=false
    Module.bUseMaxDistance=true
    Module.bUseMinDistance=false
   ) {enabled=false}
   &EmitterLifeCycle(NextLoopDuration=0.25) {enabled=false}
   &SpawnRate(
    SpawnRate=&MakeCustomFloatFromBool(Bool=$User.InWater, True Float=&Lerp_Float(A=0, B=25, Alpha=$User.ForwardSpeed), False Float=0)
    Use Spawn Probability=false
   )
   &SpawnRate(Use Spawn Probability=false) {enabled=false, module=SpawnRate001}
  ]
  particleSpawnScript [
   /Niagara/Modules/Spawn/Initialization/InitializeParticle(
    Lifetime=&Lerp_Float(A=1.8, B=&UniformRangedFloat(Minimum=0.75, Maximum=1), Alpha=$User.ForwardSpeed)
    Mesh Scale=&UniformRangedVector(Minimum=<1.25 0.8 0.8>, Maximum=<2 1.4 1.15>)
    Module.Write Color=true
    Module.Write Lifetime=true
    Module.Write Mass=true
    Module.Write Material Random=true
    Module.Write Position=true
    Module.Write Scale=true
    Module.Write SpriteRotation=false
    Module.Write SpriteSize=false
    Position=&Add_Vector(
     A=&SimulationPosition()
     B=&Multiply_VectorByFloat(Vector=$Engine.Owner.SystemXAxis, Float=-150)
    )
   )
   /SetVariables(
    Particles.MeshOrientation=&FN_OwnerRotationRollandPitchRemoved()
    Particles.Position=&MakeVector(
     X=&MakeFloatFromVector(Vector=Linked() {linkedVariable {name=Particles.Position, type=NiagaraPosition}}, Channel=::NewEnumerator0)
     Y=&MakeFloatFromVector(Vector=Linked() {linkedVariable {name=Particles.Position, type=NiagaraPosition}}, Channel=::NewEnumerator1)
     Z=$User.RearPontoonBig.WaterHeight
    )
    Particles.Scale=&Multiply_Vector(A=$Particles.Scale, B=&Lerp_Vector(A=<0.5 0.5 1>, B=<1 1 1>, Alpha=$User.ForwardSpeed))
   ) {isSetParameters=true}
   /Game/Effects/Niagara/Emitters/Water/FN_WaterDisplace(WorldPosition:NiagaraPosition=$Particles.Position) {enabled=false}
  ]
  particleUpdateScript [
   &ParticleState()
   /SetVariables(Particles.Scale=&Multiply_Vector(
     A=$Particles.Initial.Scale
     B=&VectorFromCurve(
      VectorCurve=DataInterface() {propertyValues {XCurve {keys [<at 0 1>,<at 0.9989 1>]}, YCurve {keys [<at 0 1>,<at 0.9989 1>]}, ZCurve {keys [<at 0 0>,<at 0.2345 0.7983>,<at 0.9989 1>]}}}
      CurveIndex=$Emitter.Scalability.DistanceFraction
     )
    )) {isSetParameters=true}
   &Drag(Drag=2, Use Linear Drag=true, Use Rotational Drag=false)
   &ScaleColor(
    Scale Alpha=&FloatFromCurve(FloatCurve=DataInterface() {propertyValues {Curve {keys [
         {arriveTangent=9.9926, interpMode=RCIM_Cubic, leaveTangent=9.9927, tangentMode=RCTM_User, time=0, value=0}
         {arriveTangent=-2.8339, interpMode=RCIM_Cubic, leaveTangent=-2.8339, tangentMode=RCTM_User, time=0.0216, value=1}
         <at 1.0011 0.002 RCIM_Cubic>
        ]}}})
    ScaleA=true
    ScaleRGB=true
   )
   &SolveForcesAndVelocity(Clamp Velocity=false, Limit Acceleration=false)
   /SetVariables(Particles.Scale=&Multiply_Vector(
     A=$Particles.Initial.Scale
     B=&VectorFromCurve(VectorCurve=DataInterface() {propertyValues {
        XCurve {keys [
          {arriveTangent=4.6994, interpMode=RCIM_Cubic, leaveTangent=4.6994, tangentMode=RCTM_User, time=0, value=0.0795}
          {arriveTangent=0.4589, interpMode=RCIM_Cubic, leaveTangent=0.4589, tangentMode=RCTM_User, time=0.1447, value=0.7609}
          <at 1 1>
         ]}
        YCurve {keys [{arriveTangent=2.5595, interpMode=RCIM_Cubic, leaveTangent=2.5595, tangentMode=RCTM_User, time=0, value=0.0767}, <at 1 1>]}
        ZCurve {keys [
          <at 0 0 RCIM_Cubic>
          {arriveTangent=-2.3619, interpMode=RCIM_Cubic, leaveTangent=-2.3619, tangentMode=RCTM_User, time=0.336, value=0.9567}
          <at 0.9989 0>
         ]}
       }})
    )) {isSetParameters=true}
   &DynamicMaterialParameters(
    Index 0 Param 1=&FloatFromCurve(FloatCurve=DataInterface() {propertyValues {Curve {keys [
         {arriveTangent=0.2615, interpMode=RCIM_Cubic, leaveTangent=0.2615, tangentMode=RCTM_User, time=0, value=-1.7984}
         {arriveTangent=3.2109, interpMode=RCIM_Cubic, leaveTangent=3.2109, tangentMode=RCTM_User, time=0.9985, value=0}
        ]}}})
    Index 0 Param 2=&Add_Float(
     A=&FloatFromCurve(FloatCurve=DataInterface() {propertyValues {Curve {keys [
          {arriveTangent=-9.4845, interpMode=RCIM_Cubic, leaveTangent=-9.4845, tangentMode=RCTM_User, time=0, value=4}
          {arriveTangent=0.2383, interpMode=RCIM_Cubic, leaveTangent=0.2383, tangentMode=RCTM_User, time=1, value=0}
         ]}}})
     B=0
    )
    Index 1 Param 1=&FloatFromCurve(FloatCurve=DataInterface() {propertyValues {Curve {keys [
         <at 0 0 RCIM_Cubic>
         {arriveTangent=1, interpMode=RCIM_Cubic, leaveTangent=1, time=0.5861, value=0.1787}
         {arriveTangent=3.3294, interpMode=RCIM_Cubic, leaveTangent=3.3294, tangentMode=RCTM_User, time=1, value=1}
        ]}}})
    Index 3 Param 1=&MakeFloatFromInt(Int=$User.RearPontoonBig.WaterIndex)
    Index 3 Param 2=$User.RearPontoonBig.Depth
    Module.Param0WriteEnabled=true
    Module.Param0WriteEnabled1=true
    Module.Param0WriteEnabled3=true
    Module.Param1WriteEnabled=true
    Module.Param1WriteEnabled1=true
    Module.Param1WriteEnabled3=true
    Module.Param2WriteEnabled=true
    Module.Param2WriteEnabled1=true
    Module.Param2WriteEnabled3=true
    Module.Param3WriteEnabled=true
    Module.Param3WriteEnabled1=true
    Module.Param3WriteEnabled3=true
    Write Parameter Index 1=true
    Write Parameter Index 3=true
   )
  ]
  rendererClasses [&NiagaraMeshRendererProperties]
  renderers [{rendererClass=&NiagaraMeshRendererProperties, rendererIndex=0}]
 }
 Small_Engine_Burbles_Mesh_NoSkelRef {
  bEnabled=false
  emitterUpdateScript [
   &EmitterState()
   &SpawnRate(
    SpawnRate=&MakeCustomFloatFromBool(
     Bool=&Boolean_AndOperation(Bool1=$User.HasDriver, Bool2=$User.InWater)
     True Float=&Multiply_Float(A=1, B=&Lerp_Float(A=3, B=8, Alpha=$User.ForwardSpeed))
     False Float=0
    )
    Use Spawn Probability=false
   )
  ]
  particleSpawnScript [
   /Niagara/Modules/Spawn/Initialization/V2/InitializeParticle(Color Mode=::NewEnumerator0, Lifetime=5, UsePositionOffset=false, Write Material Random=true)
   &AddVelocity(
    Coordinate Space=::Simulation
    Scale Added Velocity=&MakeVector(
     X=&Multiply_Float(A=1, B=&Lerp_Float(A=0, B=-250, Alpha=$User.ForwardSpeed))
     Y=&Multiply_Float(A=&UniformRangedFloat(Minimum=-80, Maximum=80), B=$User.ForwardSpeed)
     Z=0
    )
    Velocity=<1 1 0>
   )
   /Niagara/Modules/Spawn/Orientation/V2/InitialMeshOrientation(
    Orientation Vector=/Niagara/DynamicInputs/Random/RandomVector(
     Vector Scale:NiagaraFloat=1
     Constraint Vector:Vector3f=<0 0 1>
     Randomness Mode:ENiagaraRandomnessMode=::NewEnumerator0
     Random Seed:NiagaraInt32=0
     Fixed Random Seed:NiagaraBool=false
    )
    Use Orientation Vector=true
    Use Rotation Vector=false
   )
  ]
  particleUpdateScript [
   &ParticleState()
   &GravityForce(Gravity=/Niagara/DynamicInputs/Lerp/Lerp_VectorComponents(
     Lerp Component Alpha:Vector3f=&MakeVector(X=0, Y=0, Z=$User.ForwardVelocity)
     A:Vector3f=<0 0 -100>
     B:Vector3f=<1 1 -250>
    ))
   &Drag(Drag=3, Use Linear Drag=true, Use Rotational Drag=false)
   &ScaleColor(
    Scale Alpha=&FloatFromCurve(FloatCurve=DataInterface() {propertyValues {Curve {keys [
         {arriveTangent=0.0323, interpMode=RCIM_Cubic, leaveTangent=0.0323, tangentMode=RCTM_User, time=0, value=0}
         {arriveTangent=-0.9283, interpMode=RCIM_Cubic, leaveTangent=-0.9283, tangentMode=RCTM_User, time=0.0736, value=1}
         <at 1 0>
        ]}}})
    Scale RGB=&MakeVectorFromLinearColorRGB(Color=&ScaleLinearColorByCurve(
      InitialValue=<rgba 1 1 1 1>
      ScaleCurve=DataInterface() {propertyValues {
        AlphaCurve {keys [<at -0.0151 1>,<at 0.0979 1>]}
        BlueCurve {keys [<at 0.0097 0.5098>]}
        GreenCurve {keys [<at 0.0097 0.2039>]}
        RedCurve {keys [<at 0.0097 0>]}
       }}
      CurveIndex=$Particles.NormalizedAge
     ))
    ScaleA=true
    ScaleRGB=true
   )
   &ScaleMeshSize(Scale Factor=&VectorFromFloat(Value=&FloatFromCurve(
      FloatCurve=DataInterface() {propertyValues {Curve {keys [{arriveTangent=1.1722, interpMode=RCIM_Cubic, leaveTangent=1.1722, tangentMode=RCTM_User, time=0.0034, value=0.4567}, <at 1 1>]}}}
      CurveIndex=&Multiply_Float(
       A=1
       B=&Multiply_Float(A=$Particles.NormalizedAge, B=&Lerp_Float(A=0.15, B=1.5, Alpha=$User.ForwardVelocity))
      )
     )))
   &DynamicMaterialParameters(
    Index 0 Param 1=0.5
    Index 0 Param 2=-4.5
    Index 0 Param 3=0.2
    Module.Param0WriteEnabled=true
    Module.Param1WriteEnabled=true
    Module.Param2WriteEnabled=true
    Module.Param3WriteEnabled=true
   )
   /Niagara/Modules/Update/Orientation/V2/UpdateMeshOrientation(Rotation Rate=0.2, Rotation Vector:Vector3f=&UniformRangedVector(Minimum=<-1 -1 -1>, Maximum=<1 1 1>))
  ]
  rendererClasses [&NiagaraMeshRendererProperties, &NiagaraSpriteRendererProperties]
  renderers [
   {rendererClass=&NiagaraMeshRendererProperties, rendererIndex=0}
   {rendererClass=&NiagaraSpriteRendererProperties, rendererIndex=1}
  ]
 }
 Splash_LS {
  bEnabled=false
  emitterUpdateScript [
   &EmitterState(Life Cycle Mode=::NewEnumerator1, Loop Duration=0.25, UseLoopDelay=false)
   &EmitterScalability(
    Module.bApplySpawnCountScale=false
    Module.bApplySpawnCountScaleByDistanceFraction=false
    Module.bUseMaxDistance=true
    Module.bUseMinDistance=false
   ) {enabled=false}
   &EmitterLifeCycle(NextLoopDuration=0.25) {enabled=false}
   &SpawnRate(
    SpawnRate=&MakeCustomFloatFromBool(Bool=$User.InWater, True Float=&Lerp_Float(A=0, B=80, Alpha=$User.ForwardSpeed), False Float=0)
    Use Spawn Probability=false
   )
  ]
  particleSpawnScript [
   /Niagara/Modules/Spawn/Initialization/InitializeParticle(
    Lifetime=&Lerp_Float(
     A=&UniformRangedFloat(Minimum=0.35, Maximum=0.58)
     B=&MakeCustomFloatFromBool(Bool=$User.isBoosting, True Float=0.28, False Float=0.44)
     Alpha=$User.ForwardSpeed
    )
    Module.Write Color=true
    Module.Write Lifetime=true
    Module.Write Mass=true
    Module.Write Material Random=true
    Module.Write Position=false
    Module.Write Scale=true
    Module.Write SpriteRotation=false
    Module.Write SpriteSize=true
    Position=/Niagara/Functions/Localspace/SimulationPosition()
    Sprite Size=&UniformRangedVector2D(Minimum=<100 220>, Maximum=<50 300>)
   )
   /Game/Athena/DrivableVehicles/Meatball/Meatball_Large/FX/Modules/FN_BoatSplashInitialize(
    Distribution Range A:Vector2f=<180 300>
    Distribution Range B:Vector2f=<0 300>
    Scale Height Distribution Curve=80
    Width Offset=75
    Width Profile=DataInterface() {propertyValues {Curve {keys [
        <at 0 0.7233>
        <at 0.158 0.9654>
        {arriveTangent=-0.3891, interpMode=RCIM_Cubic, leaveTangent=-0.3891, time=0.4964, value=0.9811}
        {arriveTangent=-2.6464, interpMode=RCIM_Cubic, leaveTangent=-2.6464, tangentMode=RCTM_User, time=0.9867, value=0.6428}
       ]}}}
   )
   /Game/Athena/DrivableVehicles/Meatball/Meatball_Large/FX/Modules/FN_BoatSplashInitialSpriteSize(
    Boost Multiplier=10
    Front Pontoon Depth:NiagaraFloat=$User.FrontPontoonBig.Depth
    Front Pontoon Impulse:Vector3f=$User.FrontPontoonBig.Impulse
    Impulse Influence=DataInterface() {propertyValues {
      XCurve {keys [<at -0.0012 1.0153>,<at 0.1346 0.996>,<at 0.3563 -0.009>]}
      YCurve {keys [<at 0.2477 0.0103>,<at 0.3563 1.0346>,<at 0.6504 1.0346>,<at 0.8133 -0.0283>]}
      ZCurve {keys [<at 0.6594 -0.009>,<at 0.7907 0.996>,<at 1.0033 1.0153>]}
     }}
    Inside Bank Multiplier=1
    Mid Pontoon Impulse:Vector3f=$User.MidPontoonBig.Impulse
    MidPontoonDepth:NiagaraFloat=$User.MidPontoonBig.Depth
    Outside Bank Multiplier=3
    Rear Pontoon Depth:NiagaraFloat=$User.RearPontoonBig.Depth
    Rear Pontoon Impulse:Vector3f=$User.RearPontoonBig.Impulse
    RemapForwardSpeed=DataInterface() {propertyValues {Curve {keys [<at 0.05 0>,<at 0.9991 0.9935>]}}}
    SizeCurve=DataInterface() {propertyValues {Curve {keys [<at -0.0101 1.0359>,<at 1.0057 0.0253>]}}}
   )
   /NiagaraFN/Sprite/FN_RandomizeSpriteUV()
  ]
  particleUpdateScript [
   &ParticleState()
   /Game/Athena/DrivableVehicles/Meatball/Meatball_Large/FX/Modules/FN_BoatSplashDirectionVector(
    Distribution:NiagaraFloat=$Particles.FN_BoatSplashInitialize.RandomDistribution
    Rotate Away:Vector2f=<1.5 1.5>
    Rotate Back:Vector2f=&MakeVector2D(
     X=2
     Y=&FloatFromCurve(FloatCurve=DataInterface() {propertyValues {Curve {keys [
          <at -0.0016 0.4142 RCIM_Cubic>
          {arriveTangent=2.336, interpMode=RCIM_Cubic, leaveTangent=2.336, tangentMode=RCTM_User, time=0.9985, value=1.358}
         ]}}})
    )
   )
   /Game/Athena/DrivableVehicles/Meatball/Meatball_Large/FX/Modules/FN_BoatRemoveOwnerRoll()
   &SolveForcesAndVelocity(Clamp Velocity=false, Limit Acceleration=false)
   &ScaleSpriteSize(Scale Factor=&Multiply_Vector2(
     A=<1 0.8>
     B=&Vector2DFromFloat(Value=&Multiply_Float(
       A=1
       B=&FloatFromCurve(FloatCurve=DataInterface() {propertyValues {Curve {keys [{arriveTangent=2.411, interpMode=RCIM_Cubic, leaveTangent=2.411, tangentMode=RCTM_User, time=0, value=0.05}, <at 1.0014 1.0163>]}}})
      ))
    ))
   &Color()
   &DynamicMaterialParameters(
    Index 0 Param 1=&FloatFromCurve(FloatCurve=DataInterface() {propertyValues {Curve {keys [
         {arriveTangent=0.7433, interpMode=RCIM_Cubic, leaveTangent=0.7433, tangentMode=RCTM_User, time=0, value=-1.75}
         {arriveTangent=3.6235, interpMode=RCIM_Cubic, leaveTangent=3.6235, tangentMode=RCTM_User, time=1.0037, value=0.25}
        ]}}})
    Index 0 Param 2=&FloatFromCurve(FloatCurve=DataInterface() {propertyValues {Curve {keys [{arriveTangent=0.1759, time=0, value=0.5}, {arriveTangent=-0.3346, leaveTangent=-0.3346, time=1, value=0}]}}})
    Index 0 Param 3=&FloatFromCurve(
     FloatCurve=DataInterface() {propertyValues {Curve {keys [
         {arriveTangent=0.1191, interpMode=RCIM_Cubic, leaveTangent=0.1191, tangentMode=RCTM_User, time=0, value=0.08}
         {arriveTangent=0.0028, interpMode=RCIM_Cubic, leaveTangent=0.0028, tangentMode=RCTM_User, time=1, value=0.135}
        ]}}}
     Scale Curve=2
    )
    Index 0 Param 4=&FloatFromCurve(FloatCurve=DataInterface() {propertyValues {Curve {keys [<at 0 0>, {arriveTangent=3.3183, interpMode=RCIM_Cubic, leaveTangent=3.3183, tangentMode=RCTM_User, time=1, value=0}]}}})
    Module.Param0WriteEnabled=true
    Module.Param1WriteEnabled=true
    Module.Param2WriteEnabled=true
    Module.Param3WriteEnabled=true
   )
   /NiagaraFN/Sprite/FN_SpriteSort(SortOrder:NiagaraFloat=$Particles.FN_BoatSplashInitialize.RandomDistribution)
   &CameraOffset(Camera Offset Amount=&Multiply_Float(A=200, B=HlslExpression() {hlslExpression="abs(System.BoatRoll)"})) {enabled=false}
   /SetVariables(Particles.Lifetime=&MakeCustomFloatFromBool(Bool=$User.InAir, True Float=0, False Float=$Particles.Initial.Lifetime)) {enabled=false, isSetParameters=true, module=SetVariables_#2}
  ]
  rendererClasses [&NiagaraSpriteRendererProperties]
  renderers [{rendererClass=&NiagaraSpriteRendererProperties}]
 }
 Splashes_Mist_WS {
  bEnabled=false
  emitterUpdateScript [
   &EmitterState(Life Cycle Mode=::NewEnumerator1, Loop Duration=0.25, UseLoopDelay=false)
   &EmitterScalability(
    Module.bApplySpawnCountScale=false
    Module.bApplySpawnCountScaleByDistanceFraction=false
    Module.bUseMaxDistance=true
    Module.bUseMinDistance=false
   ) {enabled=false}
   &EmitterLifeCycle(NextLoopDuration=0.25) {enabled=false}
   &SpawnRate(
    SpawnRate=&MakeCustomFloatFromBool(
     Bool=$User.InWater
     True Float=&Lerp_Float(
      A=0
      B=120
      Alpha=&FloatFromCurve(
       FloatCurve=DataInterface() {propertyValues {Curve {keys [
           {arriveTangent=0.121, interpMode=RCIM_Cubic, leaveTangent=0.121, tangentMode=RCTM_User, time=0, value=0}
           {arriveTangent=2.9355, interpMode=RCIM_Cubic, leaveTangent=2.9355, tangentMode=RCTM_User, time=1, value=1}
          ]}}}
       CurveIndex=$User.ForwardSpeed
      )
     )
     False Float=0
    )
    Use Spawn Probability=false
   )
  ]
  particleSpawnScript [
   &InitializeParticle(
    Color=<rgba 1 1 1 0.125>
    Lifetime=&Lerp_Float(
     A=0.4
     B=&MakeCustomFloatFromBool(Bool=$User.isBoosting, True Float=0.25, False Float=0.33)
     Alpha=$User.ForwardSpeed
    )
    Mesh Scale=<10 10 10>
    Module.Write Color=true
    Module.Write Lifetime=true
    Module.Write Mass=true
    Module.Write Material Random=true
    Module.Write Position=false
    Module.Write Scale=false
    Module.Write SpriteRotation=true
    Module.Write SpriteSize=true
    Position=&SimulationPosition()
    Sprite Rotation=&UniformRangedFloat(Minimum=0.4, Maximum=360)
    Sprite Size=&UniformRangedVector2D(Minimum=<75 75>, Maximum=<90 90>)
   )
   &FN_BoatSplashInitialize(
    Distribution Range A:Vector2f=<-20 200>
    Distribution Range B:Vector2f=<0 200>
    Modulate By Length=0
    Width Offset=160
    Width Profile=DataInterface() {propertyValues {Curve {keys [
        <at 0.0043 0.7198>
        <at 0.158 0.9654>
        {arriveTangent=-0.5811, interpMode=RCIM_Cubic, leaveTangent=-0.5811, time=0.4964, value=0.9811}
        {arriveTangent=-2.6464, interpMode=RCIM_Cubic, leaveTangent=-2.6464, tangentMode=RCTM_User, time=0.9836, value=0.4856}
       ]}}}
    Width Randomize=0
   )
   &FN_BoatSplashInitialSpriteSize(
    Boost Multiplier=15
    Front Pontoon Depth:NiagaraFloat=$User.FrontPontoonBig.Depth
    Front Pontoon Impulse:Vector3f=$User.FrontPontoonBig.Impulse
    Impulse Influence=DataInterface() {propertyValues {
      XCurve {keys [<at -0.0012 1.0153>,<at 0.1346 0.996>,<at 0.3563 -0.009>]}
      YCurve {keys [<at 0.2477 0.0103>,<at 0.3563 1.0346>,<at 0.6504 1.0346>,<at 0.8133 -0.0283>]}
      ZCurve {keys [<at 0.6594 -0.009>,<at 0.7907 0.996>,<at 1.0033 1.0153>]}
     }}
    Inside Bank Multiplier=0.66
    Mid Pontoon Impulse:Vector3f=$User.MidPontoonBig.Impulse
    MidPontoonDepth:NiagaraFloat=$User.MidPontoonBig.Depth
    Outside Bank Multiplier=1.5
    Rear Pontoon Depth:NiagaraFloat=$User.RearPontoonBig.Depth
    Rear Pontoon Impulse:Vector3f=$User.RearPontoonBig.Impulse
    RemapForwardSpeed=DataInterface() {propertyValues {Curve {keys [
        {arriveTangent=2.8434, interpMode=RCIM_Cubic, leaveTangent=2.8434, tangentMode=RCTM_User, time=0.0682, value=0.015}
        {arriveTangent=0.2476, interpMode=RCIM_Cubic, leaveTangent=0.2476, tangentMode=RCTM_User, time=0.9991, value=0.9935}
       ]}}}
    SizeCurve=DataInterface() {propertyValues {Curve {keys [<at 0 1>,<at 0.3746 0.6483>]}}}
   )
   &FN_RandomizeSpriteUV()
   &FN_BoatRemoveOwnerRoll()
   &InheritVelocity(Module.Limit Speed=false, Velocity Scale=<0.08 0.08 0.225>)
   &FN_BoatSplashDirectionVector(
    Distribution:NiagaraFloat=$Particles.FN_BoatSplashInitialize.RandomDistribution
    Rotate Away:Vector2f=<0.5 0.86>
    Rotate Back:Vector2f=<1 1>
    Set Velocity=true
    Speed:NiagaraFloat=&Lerp_Float(A=80, B=450, Alpha=$User.ForwardSpeed)
   )
  ]
  particleUpdateScript [
   &ParticleState()
   &GravityForce(Gravity=<0 0 -800>) {enabled=false}
   &Drag(Drag=4.5, Use Linear Drag=true, Use Rotational Drag=false)
   &SolveForcesAndVelocity(Clamp Velocity=false, Limit Acceleration=false)
   &ScaleSpriteSize(Scale Factor=&Multiply_Vector2(
     A=<2 2>
     B=&Vector2DFromFloat(Value=&Multiply_Float(
       A=1
       B=&FloatFromCurve(FloatCurve=DataInterface() {propertyValues {Curve {keys [{arriveTangent=3.1747, interpMode=RCIM_Cubic, leaveTangent=3.1747, tangentMode=RCTM_User, time=0, value=0}, <at 1.0014 1.0163>]}}})
      ))
    ))
   &Color(
    Color=&ScaleLinearColorByCurve(
     InitialValue=<rgba 1 1 1 0.6>
     ScaleCurve=DataInterface() {propertyValues {AlphaCurve {keys [<at 0 1>]}, BlueCurve {keys [<at 0 1>]}, GreenCurve {keys [<at 0 1>]}, RedCurve {keys [<at 0 1>]}}}
     CurveIndex=$Particles.NormalizedAge
    )
    Scale Alpha=&FloatFromCurve(FloatCurve=DataInterface() {propertyValues {Curve {keys [
         <at 0 0>
         {arriveTangent=-5.647, interpMode=RCIM_Cubic, leaveTangent=-5.647, tangentMode=RCTM_User, time=0.0216, value=1.8483}
         <at 1.0011 0.002>
        ]}}})
   )
   &DynamicMaterialParameters(
    Index 0 Param 1=&FloatFromCurve(FloatCurve=DataInterface() {propertyValues {Curve {keys [
         {arriveTangent=0.3662, interpMode=RCIM_Cubic, leaveTangent=0.3662, tangentMode=RCTM_User, time=0.25, value=1}
         {arriveTangent=-1.3063, interpMode=RCIM_Cubic, leaveTangent=-1.3063, tangentMode=RCTM_User, time=1.0011, value=0}
        ]}}})
    Index 0 Param 2=&FloatFromCurve(FloatCurve=DataInterface() {propertyValues {Curve {keys [<at -0.0086 -1.3001>,<at 0.9945 1.9153>]}}})
    Index 0 Param 3=0.5
    Module.Param0WriteEnabled=true
    Module.Param1WriteEnabled=false
    Module.Param2WriteEnabled=true
    Module.Param3WriteEnabled=true
   )
   &FN_SpriteSort(SortOrder:NiagaraFloat=$Particles.FN_BoatSplashInitialize.RandomDistribution) {enabled=false}
   &CameraOffset(Camera Offset Amount=&Multiply_Float(A=200, B=HlslExpression() {hlslExpression="abs(System.BoatRoll)"})) {enabled=false}
   /SetVariables(Particles.Lifetime=&MakeCustomFloatFromBool(Bool=$User.InAir, True Float=0, False Float=$Particles.Initial.Lifetime)) {enabled=false, isSetParameters=true, module=SetVariables_#2}
   &SpriteRotationRate(Rotation Rate=&UniformRangedFloat(Minimum=-3, Maximum=-3))
   /Niagara/Modules/Update/SubUV/SubUVAnimation(Number Of Frames=32)
  ]
  rendererClasses [&NiagaraMeshRendererProperties, &NiagaraSpriteRendererProperties]
  renderers [
   {rendererClass=&NiagaraMeshRendererProperties, rendererIndex=0}
   {rendererClass=&NiagaraSpriteRendererProperties, rendererIndex=1}
  ]
 }
 Splashes_Watery_WS {
  bEnabled=false
  emitterUpdateScript [
   &EmitterState(Life Cycle Mode=::NewEnumerator1, Loop Duration=0.25, UseLoopDelay=false)
   &EmitterScalability(
    Module.bApplySpawnCountScale=false
    Module.bApplySpawnCountScaleByDistanceFraction=false
    Module.bUseMaxDistance=true
    Module.bUseMinDistance=false
   ) {enabled=false}
   &EmitterLifeCycle(NextLoopDuration=0.25) {enabled=false}
   &SpawnRate(
    SpawnRate=&MakeCustomFloatFromBool(
     Bool=$User.InWater
     True Float=&Lerp_Float(
      A=0
      B=120
      Alpha=&FloatFromCurve(
       FloatCurve=DataInterface() {propertyValues {Curve {keys [
           {arriveTangent=0.121, interpMode=RCIM_Cubic, leaveTangent=0.121, tangentMode=RCTM_User, time=0, value=0}
           {arriveTangent=2.9355, interpMode=RCIM_Cubic, leaveTangent=2.9355, tangentMode=RCTM_User, time=1, value=1}
          ]}}}
       CurveIndex=$User.ForwardSpeed
      )
     )
     False Float=0
    )
    Use Spawn Probability=false
   )
  ]
  particleSpawnScript [
   &InitializeParticle(
    Lifetime=&Lerp_Float(
     A=0.4
     B=&MakeCustomFloatFromBool(Bool=$User.isBoosting, True Float=0.25, False Float=0.33)
     Alpha=$User.ForwardSpeed
    )
    Mesh Scale=<10 10 10>
    Module.Write Color=true
    Module.Write Lifetime=true
    Module.Write Mass=true
    Module.Write Material Random=true
    Module.Write Position=false
    Module.Write Scale=true
    Module.Write SpriteRotation=true
    Module.Write SpriteSize=true
    Position=&SimulationPosition()
    Sprite Rotation=&UniformRangedFloat(Minimum=0.4, Maximum=360)
    Sprite Size=&UniformRangedVector2D(Minimum=<65 65>, Maximum=<50 50>)
   )
   &FN_BoatSplashInitialize(
    Distribution Range A:Vector2f=<-20 200>
    Distribution Range B:Vector2f=<0 200>
    Modulate By Length=0
    Width Offset=160
    Width Profile=DataInterface() {propertyValues {Curve {keys [
        <at 0.0043 0.7198>
        <at 0.158 0.9654>
        {arriveTangent=-0.5811, interpMode=RCIM_Cubic, leaveTangent=-0.5811, time=0.4964, value=0.9811}
        {arriveTangent=-2.6464, interpMode=RCIM_Cubic, leaveTangent=-2.6464, tangentMode=RCTM_User, time=0.9836, value=0.4856}
       ]}}}
    Width Randomize=0
   )
   &FN_BoatSplashInitialSpriteSize(
    Boost Multiplier=10
    Front Pontoon Depth:NiagaraFloat=$User.FrontPontoonBig.Depth
    Front Pontoon Impulse:Vector3f=$User.FrontPontoonBig.Impulse
    Impulse Influence=DataInterface() {propertyValues {
      XCurve {keys [<at -0.0012 1.0153>,<at 0.1346 0.996>,<at 0.3563 -0.009>]}
      YCurve {keys [<at 0.2477 0.0103>,<at 0.3563 1.0346>,<at 0.6504 1.0346>,<at 0.8133 -0.0283>]}
      ZCurve {keys [<at 0.6594 -0.009>,<at 0.7907 0.996>,<at 1.0033 1.0153>]}
     }}
    Inside Bank Multiplier=0.66
    Mid Pontoon Impulse:Vector3f=$User.MidPontoonBig.Impulse
    MidPontoonDepth:NiagaraFloat=$User.MidPontoonBig.Depth
    Outside Bank Multiplier=1.5
    Rear Pontoon Depth:NiagaraFloat=$User.RearPontoonBig.Depth
    Rear Pontoon Impulse:Vector3f=$User.RearPontoonBig.Impulse
    RemapForwardSpeed=DataInterface() {propertyValues {Curve {keys [
        {arriveTangent=0.029, interpMode=RCIM_Cubic, leaveTangent=0.029, tangentMode=RCTM_User, time=0.15, value=0}
        {arriveTangent=2.8118, interpMode=RCIM_Cubic, leaveTangent=2.8118, tangentMode=RCTM_User, time=0.9991, value=0.9935}
       ]}}}
    SizeCurve=DataInterface() {propertyValues {Curve {keys [<at 0 1>,<at 0.3746 0.6483>]}}}
   )
   &FN_RandomizeSpriteUV()
   &FN_BoatRemoveOwnerRoll()
   &InheritVelocity(Module.Limit Speed=false, Velocity Scale=<0.08 0.08 0.225>)
   &SubUVAnimation(SubUV Lookup Index=&UniformRangedFloat(Minimum=0, Maximum=1))
   &FN_BoatSplashDirectionVector(
    Distribution:NiagaraFloat=$Particles.FN_BoatSplashInitialize.RandomDistribution
    Rotate Away:Vector2f=<0.5 0.86>
    Rotate Back:Vector2f=<1 1>
    Set Velocity=true
    Speed:NiagaraFloat=&Lerp_Float(A=50, B=450, Alpha=$User.ForwardSpeed)
   )
  ]
  particleUpdateScript [
   &ParticleState()
   &GravityForce(Gravity=<0 0 -800>) {enabled=false}
   &Drag(Drag=2, Use Linear Drag=true, Use Rotational Drag=false)
   &SolveForcesAndVelocity(Clamp Velocity=false, Limit Acceleration=false)
   &ScaleSpriteSize(Scale Factor=&Multiply_Vector2(
     A=<2 2>
     B=&Vector2DFromFloat(Value=&Multiply_Float(
       A=1
       B=&FloatFromCurve(FloatCurve=DataInterface() {propertyValues {Curve {keys [{arriveTangent=3.1747, interpMode=RCIM_Cubic, leaveTangent=3.1747, tangentMode=RCTM_User, time=0, value=0}, <at 1.0014 1.0163>]}}})
      ))
    ))
   &Color(
    Color=&ScaleLinearColorByCurve(
     InitialValue=<rgba 1 1 1 1>
     ScaleCurve=DataInterface() {propertyValues {
       AlphaCurve {keys [<at 0.1803 1>,<at 0.7614 1>]}
       BlueCurve {keys [<at 0.5923 1>]}
       GreenCurve {keys [<at 0.5923 1>]}
       RedCurve {keys [<at 0.5923 1>]}
      }}
     CurveIndex=$Particles.NormalizedAge
    )
    Scale Alpha=&FloatFromCurve(FloatCurve=DataInterface() {propertyValues {Curve {keys [<at 0 0>,<at 0.05 1>,<at 1.0011 0.002>]}}})
   )
   &DynamicMaterialParameters(
    Index 0 Param 1=&FloatFromCurve(FloatCurve=DataInterface() {propertyValues {Curve {keys [
         {arriveTangent=0.3662, interpMode=RCIM_Cubic, leaveTangent=0.3662, tangentMode=RCTM_User, time=0.25, value=1}
         {arriveTangent=-1.3063, interpMode=RCIM_Cubic, leaveTangent=-1.3063, tangentMode=RCTM_User, time=1.0011, value=0}
        ]}}})
    Index 0 Param 2=&FloatFromCurve(FloatCurve=DataInterface() {propertyValues {Curve {keys [<at -0.0086 -1.3001>,<at 0.9945 1.9153>]}}})
    Index 0 Param 3=0.5
    Module.Param0WriteEnabled=true
    Module.Param1WriteEnabled=false
    Module.Param2WriteEnabled=true
    Module.Param3WriteEnabled=true
   )
   &FN_SpriteSort(SortOrder:NiagaraFloat=$Particles.FN_BoatSplashInitialize.RandomDistribution) {enabled=false}
   &CameraOffset(Camera Offset Amount=&Multiply_Float(A=200, B=HlslExpression() {hlslExpression="abs(System.BoatRoll)"})) {enabled=false}
   /SetVariables(Particles.Lifetime=&MakeCustomFloatFromBool(Bool=$User.InAir, True Float=0, False Float=$Particles.Initial.Lifetime)) {enabled=false, isSetParameters=true, module=SetVariables_#2}
   &SpriteRotationRate(Rotation Rate=&UniformRangedFloat(Minimum=-3, Maximum=-3))
  ]
  rendererClasses [&NiagaraMeshRendererProperties, &NiagaraSpriteRendererProperties]
  renderers [
   {rendererClass=&NiagaraMeshRendererProperties, rendererIndex=0}
   {rendererClass=&NiagaraSpriteRendererProperties, rendererIndex=1}
  ]
 }
}
summary {
 systemName=NS_Vehicle_Meatball_Water_Clyde
 emitters [
  {emitterName=FN_Sprite_SpeedLines, rendererClasses [&NiagaraSpriteRendererProperties]}
  {bEnabled=false, emitterName=Splash_LS, rendererClasses [&NiagaraSpriteRendererProperties]}
  {
   bEnabled=false
   emitterName=Splashes_Watery_WS
   rendererClasses [&NiagaraMeshRendererProperties, &NiagaraSpriteRendererProperties]
  }
  {emitterName=FN_Water_Foam_SnapToSurface_NoSkelRef, rendererClasses [&NiagaraMeshRendererProperties]}
  {emitterName=Mesh_Splashes_LS, rendererClasses [&NiagaraMeshRendererProperties]}
  {
   bEnabled=false
   emitterName=Splashes_Mist_WS
   rendererClasses [&NiagaraMeshRendererProperties, &NiagaraSpriteRendererProperties]
  }
  {
   bEnabled=false
   emitterName=Small_Engine_Burbles_Mesh_NoSkelRef
   rendererClasses [&NiagaraMeshRendererProperties, &NiagaraSpriteRendererProperties]
  }
  {emitterName=Motor_Burble_Mesh_WS_NoSkelRef, rendererClasses [&NiagaraMeshRendererProperties]}
  {emitterName=Mesh_Splashes_LS001, rendererClasses [&NiagaraMeshRendererProperties]}
 ]
 userVariables [
  User.Boost {type=NiagaraFloat, defaultValue=0, description=""}
  User.DebugWaterIndex {type=NiagaraInt32}
  User.DefaultWakeStrength {type=NiagaraFloat, defaultValue=0, description=""}
  User.EnableSpawn {type=NiagaraBool}
  User.ForwardSpeed {type=NiagaraFloat, defaultValue=0, description=""}
  User.ForwardVelocity {type=NiagaraFloat, defaultValue=0, description=""}
  User.FrontPontoonBig.Depth {type=NiagaraFloat, defaultValue=1, description=""}
  User.FrontPontoonBig.Impulse {type=Vector3f, description="", defaultValue {struct=Vector3f, value=<0 0 0>}}
  User.FrontPontoonBig.Location {type=Vector3f, description="", defaultValue {struct=Vector3f, value=<0 0 0>}}
  User.FrontPontoonBig.Rotation {type=Quat4f, description="", defaultValue {struct=Quat4f, value=<0 0 0 1>}}
  User.FrontPontoonBig.WaterHeight {type=NiagaraFloat, defaultValue=0, description=""}
  User.HasDriver {type=NiagaraBool}
  User.ImmersionDepth {type=NiagaraFloat, defaultValue=0, description=""}
  User.Impulse {type=Vector3f, description="", defaultValue {struct=Vector3f, value=<0 0 0>}}
  User.InAir {type=NiagaraBool}
  User.InWater {type=NiagaraBool}
  User.isBoosting {type=NiagaraBool}
  User.IsInWaterCenter {type=NiagaraFloat, defaultValue=0, description=""}
  User.Lerp Component Alpha.Gravity {type=Vector3f, description="", defaultValue {struct=Vector3f, value=<0 0 0>}}
  User.MidPontoonBig.Depth {type=NiagaraFloat, defaultValue=0, description=""}
  User.MidPontoonBig.Impulse {type=Vector3f, description="", defaultValue {struct=Vector3f, value=<0 0 0>}}
  User.MidPontoonBig.Location {type=NiagaraPosition, description="", defaultValue {}}
  User.MidPontoonBig.Rotation {type=Quat4f, description="", defaultValue {struct=Quat4f, value=<0 0 0 1>}}
  User.MidPontoonBig.WaterBodyIndex {type=NiagaraFloat, defaultValue=0, description=""}
  User.MidPontoonBig.WaterHeight {type=NiagaraFloat, defaultValue=0, description=""}
  User.MidPontoonBig.WaterPlaneLocation {type=Vector3f, description="", defaultValue {struct=Vector3f, value=<0 0 0>}}
  User.MidPontoonBig.WaterPlaneNormal {type=Vector3f, description="", defaultValue {struct=Vector3f, value=<0 0 0>}}
  User.NewVariable {type=NiagaraFloat, defaultValue=0, description=""}
  User.NormalizedSpeed {type=NiagaraFloat, defaultValue=0, description=""}
  User.PontoonDepth {type=NiagaraFloat, defaultValue=0, description=""}
  User.PontoonLocation {type=Vector3f, description="", defaultValue {struct=Vector3f, value=<0 0 0>}}
  User.PontoonRotation {type=Quat4f, description="", defaultValue {struct=Quat4f, value=<0 0 0 1>}}
  User.RearPontoonBig.Depth {type=NiagaraFloat, defaultValue=0, description=""}
  User.RearPontoonBig.Impulse {type=Vector3f, description="", defaultValue {struct=Vector3f, value=<0 0 0>}}
  User.RearPontoonBig.Location {type=Vector3f, description="", defaultValue {struct=Vector3f, value=<0 0 0>}}
  User.RearPontoonBig.Rotation {type=Quat4f, description="", defaultValue {struct=Quat4f, value=<0 0 0 1>}}
  User.RearPontoonBig.WaterBodyIndex {type=NiagaraFloat, defaultValue=0, description=""}
  User.RearPontoonBig.WaterHeight {type=NiagaraFloat, defaultValue=0, description=""}
  User.RearPontoonBig.WaterIndex {type=NiagaraInt32}
  User.RearPontoonBig.WaterPlaneLocation {type=Vector3f, description="", defaultValue {struct=Vector3f, value=<0 0 0>}}
  User.RearPontoonBig.WaterPlaneNormal {type=Vector3f, description="", defaultValue {struct=Vector3f, value=<0 0 0>}}
  User.TurnDirection {type=NiagaraFloat, defaultValue=0, description=""}
  User.TurningAngle {type=NiagaraFloat, defaultValue=0, description=""}
  User.TurningSpeed {type=NiagaraFloat, defaultValue=0, description=""}
  User.TurnSpeed {type=NiagaraFloat, defaultValue=0, description=""}
  User.WaterBodyIndex {type=NiagaraFloat, defaultValue=0, description=""}
  User.WaterDepth {type=NiagaraFloat, defaultValue=0, description=""}
  User.WaterHeight {type=NiagaraFloat, defaultValue=0, description=""}
  User.WaterPlaneLocation {type=Vector3f, description="", defaultValue {struct=Vector3f, value=<0 0 0>}}
  User.WaterPlaneNormal {type=Vector3f, description="", defaultValue {struct=Vector3f, value=<0 0 0>}}
  User.WaterSurfaceVelocty {type=Vector3f, description="", defaultValue {struct=Vector3f, value=<0 0 0>}}
  User.WaveAttenuationFactor {type=NiagaraFloat, defaultValue=0, description=""}
  User.WaveHeight {type=NiagaraFloat, defaultValue=0, description=""}
  User.WaveTime {type=NiagaraFloat, defaultValue=0, description=""}
 ]
}
